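# Niagara effect for the SilverBell emote: warm golden glow, light rays and sparkles.
# One-shot sparkle bursts open it, then delayed looping bling and ray bursts run until a delayed ending ray, alongside rotating ground-light meshes.
# An inner glow mesh and a waveform-flickered airglow sprite with light last User.Duration.
# Time-of-day factors dim glow and bling; the three emoji emitters are disabled.
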NiagaraSystem /BRCosmetics/Animation/Game/MainPlayer/Emotes/SilverBell/FX/NS_SilverbellEmote_FX
dependencies {
 usedDataInterfaces [
  &NiagaraDataInterfaceCurve
  &NiagaraDataInterfaceColorCurve
  &NiagaraDataInterfaceVectorCurve
  &NiagaraDataInterfaceSpriteRendererInfo
  &NiagaraDataInterfaceArrayFloat
 ]
 usedDynamicInputs [&MakeVector, &RotateVectorByQuaternion]
 usedModules [
  &SystemState
  &self:SystemSpawnScript.SystemScriptSource.SystemScriptGraph.NiagaraNodeAssignment_0.SetVariables_#1
 ]
}
emitters {
 Airglow {
  emitterUpdateScript [
   &EmitterState()
   &SpawnBurst_Instantaneous(Use Spawn Probability=false, UseLoopCountLimit=false)
   /Niagara/Noise/Waveform(
    EnableExponent1=false
    GlobalClampMax=false
    GlobalClampMin=false
    UseInput1=false
    UsePhase1=false
    "Waveform [1]":ENiagara_Waveforms=::NewEnumerator12
    "[1] Frequency"=7
   )
   /SetVariables(Emitter.Flicker:NiagaraFloat=&Add_Float(A=&Multiply_Float(A=$Emitter.Waveform.Waveform, B=0.5), B=0.5)) {isSetParameters=true}
  ]
  particleSpawnScript [
   &InitializeParticle(
    Color=<rgba 1 0.3963 0 0.7>
    Lifetime=$User.Duration
    Position Offset=<-10 0 0>
    Position Offset Coordinate Space=::Local
    Sprite Size Mode=::NewEnumerator3
    Uniform Sprite Size=120
    UsePositionOffset=true
    Write Material Random=true
   )
   &DynamicMaterialParameters(
    Index 0 Param 1=0.2
    Index 0 Param 2=1
    Module.Param0WriteEnabled=true
    Module.Param1WriteEnabled=true
    Module.Param2WriteEnabled=true
    Module.Param3WriteEnabled=true
   )
   /SetVariables(Particles.DY0:Vector4f=<2 30 0 0>) {isSetParameters=true}
  ]
  particleUpdateScript [
   &ParticleState()
   &FN_Sprite_Attributes(
    Color=&MultiplyLinearColor(
     Linear Color=<rgba 1.25 0.8333 0.25 1.5>
     Scale Factor=&MakeVector4(
      X=1
      Y=1
      Z=1
      W=&FloatFromCurve(FloatCurve=DataInterface() {propertyValues {Curve {keys [<at 0 0>,<at 0.0001 1>,<at 0.79 0.85>,<at 0.8 0>]}}}, Scale Curve=$System.TOD)
     )
    )
    Create Color Attribute=true
    Create Position Attribute=true
    Create SpriteSize Attribute=true
    SpriteSize=&Multiply_Vector2D(
     A=<50 50>
     B=&Vector2DFromFloat(Value=&FloatFromCurve(
       FloatCurve=DataInterface() {propertyValues {Curve {keys [{arriveTangent=3.3159, interpMode=RCIM_Cubic, leaveTangent=3.3159, tangentMode=RCTM_User, time=0, value=0.35}, <at 0.02 1>]}}}
       Scale Curve=2
      ))
    )
    Write to Module Namespace=true
   )
   &ScaleColor(Scale Alpha=&Multiply_Float(A=$Emitter.Flicker, B=0.3), ScaleA=true, ScaleRGB=true)
   &FN_Light_Attributes(
    Color=$Particles.Color
    Create Color Attribute=true
    Create Exponent Attribute=true
    Create Radius Attribute=true
    Exponent=5
    Radius=100
    Scale Color=&Multiply_Float(
     A=&Multiply_Float(
      A=7
      B=&FloatFromCurve(FloatCurve=DataInterface() {propertyValues {Curve {keys [<at 0 0>,<at 0.001 1>,<at 0.79 0.85>,<at 0.8 0>]}}})
     )
     B=$Emitter.Flicker
    )
   )
   &ScaleColor(
    Scale Alpha=&FloatFromCurve(FloatCurve=DataInterface() {propertyValues {Curve {keys [<at 0 0>,<at 0.0001 1>,<at 0.79 0.85>,<at 0.8 0>]}}}, Scale Curve=$System.TOD)
    ScaleA=true
    ScaleRGB=true
   ) {module=ScaleColor001}
  ]
  rendererClasses [&NiagaraSpriteRendererProperties, &NiagaraLightRendererProperties]
  renderers [
   {rendererClass=&NiagaraSpriteRendererProperties}
   {rendererClass=&NiagaraLightRendererProperties, rendererIndex=1}
   {rendererClass=&NiagaraSpriteRendererProperties, rendererIndex=2}
  ]
 }
 Bling_Big {
  emitterUpdateScript [
   &EmitterState(
    Delay First Loop Only=true
    Life Cycle Mode=::NewEnumerator1
    Loop Behavior=::NewEnumerator2
    Loop Count=45
    Loop Delay=2.2
    Loop Duration=0.2
    UseLoopDelay=true
   )
   &SpawnBurst_Instantaneous(
    Spawn Count=&RandomRangeInt(Minimum=3, Maximum=6, Evaluation Type=::NewEnumerator1)
    Spawn Time=0.15
    Use Spawn Probability=false
   )
  ]
  particleSpawnScript [
   &FN_SetSpawnGroup(Number of Spawn Groups=6)
   &InitializeParticle(
    Color=<rgba 10 3.9582 0 1>
    Lifetime=&RandomRangeFloat(Minimum=0.3, Maximum=0.5)
    Position Offset=<0 0 20>
    Position Offset Coordinate Space=::Local
    Sprite Rotation Angle=&SelectFloatFromArray(
     Array Sampling Mode=::NewEnumerator0
     Direct Array Index=?
     Float Selection Array=DataInterface() {propertyValues {FloatData [0,0.25]}}
     Random Seed=0
    )
    Sprite Rotation Mode=::NewEnumerator1
    Sprite Size Mode=::NewEnumerator1
    Uniform Sprite Size Max=40
    Uniform Sprite Size Min=20
    UsePositionOffset=true
    Write Material Random=true
   )
   &ShapeLocation(MaskSpawnGroup=false, Override Seed=false, Sphere Radius=40, UseOwnerScale=true)
   &AddVelocity(Module.Use Falloff Distance=false, Override Seed=false, Velocity Mode=::NewEnumerator1, Velocity Speed=400) {enabled=false}
   /SetVariables(Particles.InitialSize:Vector2f=&Vector2DFromFloat(Value=&RandomRangeFloat(Minimum=40, Maximum=80))) {isSetParameters=true, module=SetVariables_#2}
   /SetVariables(Particles.SpriteRotation2:NiagaraFloat=&RandomRangeFloat(Minimum=0, Maximum=360)) {isSetParameters=true, module=SetVariables_#3}
   &CameraOffset(Camera Offset Amount=&RandomRangeFloat(Minimum=20, Maximum=40))
  ]
  particleUpdateScript [
   &Drag(Drag=&RandomRangeFloat(Minimum=10, Maximum=15), Use Linear Drag=true, Use Rotational Drag=false)
   &GravityForce(Gravity=&RandomRangeVector(Minimum=<0 0 -200>, Maximum=<1 1 -400>))
   &SolveForcesAndVelocity(Clamp Velocity=false, Limit Acceleration=false)
   &ParticleState()
   /SetVariables(Particles.TimeOfDay:NiagaraFloat=&FNTimeOfDayScalar(Night Scale=0.5, Morning Scale=0.3, Evening Scale=0.3)) {isSetParameters=true, module=SetVariables_#4}
   /SetVariables(Particles.Color2=&MultiplyLinearColor(
     Linear Color=&ColorFromCurve(ColorCurve=DataInterface() {propertyValues {
        AlphaCurve {keys [<at 0 0>,<at 0.1043 0.2 RCIM_Cubic>,<at 1 0 RCIM_Cubic>]}
        BlueCurve {keys [<at 0.1483 0.1562 RCIM_Cubic>,<at 1 0.0156 RCIM_Cubic>]}
        GreenCurve {keys [<at 0.1483 0.7188 RCIM_Cubic>,<at 1 0.5898 RCIM_Cubic>]}
        RedCurve {keys [<at 0.1483 1 RCIM_Cubic>,<at 1 1 RCIM_Cubic>]}
       }})
     Scale Factor=&MakeFloat4FromVectorAndFloat("Vector (RGB)"=<0.25 0.25 0.25>, "Float (Alpha)"=$Particles.TimeOfDay)
    )) {isSetParameters=true, module=SetVariables_#5}
   /SetVariables(Particles.SpriteSize2=&Multiply_Vector2(
     A=$Particles.InitialSize
     B=&Vector2DFromFloat(Value=&FloatFromCurve(
       FloatCurve=DataInterface() {propertyValues {Curve {keys [
           <at 0 0 RCIM_Cubic>
           {arriveTangent=-0.0175, interpMode=RCIM_Cubic, leaveTangent=-0.0175, tangentMode=RCTM_User, time=0.1563, value=0.907}
           {arriveTangent=-0.0042, interpMode=RCIM_Cubic, leaveTangent=-0.0042, tangentMode=RCTM_User, time=0.2646, value=0.0384}
           {arriveTangent=0.0081, interpMode=RCIM_Cubic, leaveTangent=0.0081, tangentMode=RCTM_User, time=0.4692, value=0.7123}
           {arriveTangent=0.034, interpMode=RCIM_Cubic, leaveTangent=0.034, tangentMode=RCTM_User, time=0.6406, value=0.0348}
           {arriveTangent=0.019, interpMode=RCIM_Cubic, leaveTangent=0.019, tangentMode=RCTM_User, time=0.8, value=0.25}
           <at 1 0 RCIM_Cubic>
          ]}}}
       Scale Curve=0.5
      ))
    )) {isSetParameters=true, module=SetVariables_#6}
   &ScaleSpriteSize(Scale Factor=&Vector2DFromFloat(Value=&FloatFromCurve(
      FloatCurve=DataInterface() {propertyValues {Curve {keys [
          <at 0 0 RCIM_Cubic>
          {arriveTangent=-0.0175, interpMode=RCIM_Cubic, leaveTangent=-0.0175, tangentMode=RCTM_User, time=0.1563, value=0.907}
          {arriveTangent=-0.0042, interpMode=RCIM_Cubic, leaveTangent=-0.0042, tangentMode=RCTM_User, time=0.2646, value=0.0384}
          {arriveTangent=0.0081, interpMode=RCIM_Cubic, leaveTangent=0.0081, tangentMode=RCTM_User, time=0.4692, value=0.7123}
          {arriveTangent=0.034, interpMode=RCIM_Cubic, leaveTangent=0.034, tangentMode=RCTM_User, time=0.6406, value=0.0348}
          {arriveTangent=0.019, interpMode=RCIM_Cubic, leaveTangent=0.019, tangentMode=RCTM_User, time=0.7993, value=0.5053}
          <at 1 0 RCIM_Cubic>
         ]}}}
      Scale Curve=0.5
     ))) {module=ScaleSpriteSize001}
   &ScaleColor(
    Scale Alpha=$Particles.TimeOfDay
    Scale RGB=&Multiply_Vector(
     A=<1 1 1>
     B=&VectorFromFloat(Value=&FloatFromCurve(FloatCurve=DataInterface() {propertyValues {Curve {keys [<at 0 0>,<at 0.1 1>,<at 1 0>]}}}))
    )
    ScaleA=true
    ScaleRGB=true
   )
  ]
  rendererClasses [&NiagaraSpriteRendererProperties]
  renderers [
   {rendererClass=&NiagaraSpriteRendererProperties}
   {rendererClass=&NiagaraSpriteRendererProperties, rendererIndex=1}
  ]
 }
 Bling_Big001 {
  emitterUpdateScript [
   &EmitterState(
    Delay First Loop Only=true
    Life Cycle Mode=::NewEnumerator1
    Loop Behavior=::NewEnumerator2
    Loop Count=45
    Loop Delay=2.2
    Loop Duration=0.2
    UseLoopDelay=true
   )
   &SpawnBurst_Instantaneous(
    Spawn Count=&RandomRangeInt(Minimum=1, Maximum=3, Evaluation Type=::NewEnumerator1)
    Spawn Time=0.15
    Use Spawn Probability=false
   )
  ]
  particleSpawnScript [
   &FN_SetSpawnGroup(Number of Spawn Groups=6)
   &InitializeParticle(
    Color=&SelectLinearColorFromArray(
     Array Sampling Mode=::NewEnumerator1
     Direct Array Index=&RandomRangeInt(Minimum=0, Maximum=2)
     Color Selection Array=$User.Color Array
     Random Seed=?
    )
    Lifetime=&RandomRangeFloat(Minimum=0.3, Maximum=0.5)
    Position Offset=<0 0 20>
    Position Offset Coordinate Space=::Local
    Sprite Rotation Angle=&SelectFloatFromArray(
     Array Sampling Mode=::NewEnumerator0
     Direct Array Index=?
     Float Selection Array=DataInterface() {propertyValues {FloatData [0,0.25]}}
     Random Seed=0
    )
    Sprite Rotation Mode=::NewEnumerator1
    Sprite Size Mode=::NewEnumerator1
    Uniform Sprite Size Max=40
    Uniform Sprite Size Min=20
    UsePositionOffset=true
    Write Material Random=true
   )
   &ShapeLocation(MaskSpawnGroup=false, Override Seed=false, Sphere Radius=40, UseOwnerScale=true)
   &AddVelocity(Module.Use Falloff Distance=false, Override Seed=false, Velocity Mode=::NewEnumerator1, Velocity Speed=400) {enabled=false}
   /SetVariables(Particles.InitialSize:Vector2f=&Vector2DFromFloat(Value=&RandomRangeFloat(Minimum=40, Maximum=80))) {isSetParameters=true, module=SetVariables_#2}
   /SetVariables(Particles.SpriteRotation2:NiagaraFloat=&RandomRangeFloat(Minimum=0, Maximum=360)) {isSetParameters=true, module=SetVariables_#3}
   &CameraOffset(Camera Offset Amount=&RandomRangeFloat(Minimum=20, Maximum=40))
  ]
  particleUpdateScript [
   &Drag(Drag=&RandomRangeFloat(Minimum=10, Maximum=15), Use Linear Drag=true, Use Rotational Drag=false)
   &GravityForce(Gravity=&RandomRangeVector(Minimum=<0 0 -200>, Maximum=<1 1 -400>))
   &SolveForcesAndVelocity(Clamp Velocity=false, Limit Acceleration=false)
   &ParticleState()
   &ScaleSpriteSize(
    Uniform Curve Scale=0.5
    Uniform Curve Sprite Scale=DataInterface() {propertyValues {Curve {keys [{tangentMode=RCTM_User, time=0, value=1}, {tangentMode=RCTM_User, time=1, value=0}]}}}
   )
   /SetVariables(Particles.TimeOfDay:NiagaraFloat=&FNTimeOfDayScalar(Night Scale=0.5, Morning Scale=0.3, Evening Scale=0.3)) {isSetParameters=true, module=SetVariables_#4}
   /SetVariables(Particles.Color2=&MultiplyLinearColor(
     Linear Color=&ColorFromCurve(ColorCurve=DataInterface() {propertyValues {
        AlphaCurve {keys [<at 0 0>,<at 0.1043 0.2 RCIM_Cubic>,<at 1 0 RCIM_Cubic>]}
        BlueCurve {keys [<at 0.1483 0.1562 RCIM_Cubic>,<at 1 0.0156 RCIM_Cubic>]}
        GreenCurve {keys [<at 0.1483 0.7188 RCIM_Cubic>,<at 1 0.5898 RCIM_Cubic>]}
        RedCurve {keys [<at 0.1483 1 RCIM_Cubic>,<at 1 1 RCIM_Cubic>]}
       }})
     Scale Factor=&MakeFloat4FromVectorAndFloat("Vector (RGB)"=<0.25 0.25 0.25>, "Float (Alpha)"=$Particles.TimeOfDay)
    )) {isSetParameters=true, module=SetVariables_#5}
   /SetVariables(Particles.SpriteSize2=&Multiply_Vector2(
     A=$Particles.InitialSize
     B=&Vector2DFromFloat(Value=&FloatFromCurve(
       FloatCurve=DataInterface() {propertyValues {Curve {keys [
           <at 0 0 RCIM_Cubic>
           {arriveTangent=-0.0175, interpMode=RCIM_Cubic, leaveTangent=-0.0175, tangentMode=RCTM_User, time=0.1563, value=0.907}
           {arriveTangent=-0.0042, interpMode=RCIM_Cubic, leaveTangent=-0.0042, tangentMode=RCTM_User, time=0.2646, value=0.0384}
           {arriveTangent=0.0081, interpMode=RCIM_Cubic, leaveTangent=0.0081, tangentMode=RCTM_User, time=0.4692, value=0.7123}
           {arriveTangent=0.034, interpMode=RCIM_Cubic, leaveTangent=0.034, tangentMode=RCTM_User, time=0.6406, value=0.0348}
           {arriveTangent=0.019, interpMode=RCIM_Cubic, leaveTangent=0.019, tangentMode=RCTM_User, time=0.8, value=0.25}
           <at 1 0 RCIM_Cubic>
          ]}}}
       Scale Curve=0.5
      ))
    )) {enabled=false, isSetParameters=true, module=SetVariables_#6}
   &ScaleSpriteSize(Scale Factor=&Vector2DFromFloat(Value=&FloatFromCurve(
      FloatCurve=DataInterface() {propertyValues {Curve {keys [
          <at 0 0 RCIM_Cubic>
          {arriveTangent=-0.0175, interpMode=RCIM_Cubic, leaveTangent=-0.0175, tangentMode=RCTM_User, time=0.1563, value=0.907}
          {arriveTangent=-0.0042, interpMode=RCIM_Cubic, leaveTangent=-0.0042, tangentMode=RCTM_User, time=0.2646, value=0.0384}
          {arriveTangent=0.0081, interpMode=RCIM_Cubic, leaveTangent=0.0081, tangentMode=RCTM_User, time=0.4692, value=0.7123}
          {arriveTangent=0.034, interpMode=RCIM_Cubic, leaveTangent=0.034, tangentMode=RCTM_User, time=0.6406, value=0.0348}
          {arriveTangent=0.019, interpMode=RCIM_Cubic, leaveTangent=0.019, tangentMode=RCTM_User, time=0.7993, value=0.5053}
          <at 1 0 RCIM_Cubic>
         ]}}}
      Scale Curve=0.5
     ))) {enabled=false, module=ScaleSpriteSize001}
   &ScaleColor(
    Scale Alpha=$Particles.TimeOfDay
    Scale RGB=&Multiply_Vector(
     A=<1 1 1>
     B=&VectorFromFloat(Value=&FloatFromCurve(FloatCurve=DataInterface() {propertyValues {Curve {keys [<at 0 0>,<at 0.1 1>,<at 1 0>]}}}))
    )
    ScaleA=true
    ScaleRGB=true
   )
  ]
  rendererClasses [&NiagaraSpriteRendererProperties]
  renderers [
   {rendererClass=&NiagaraSpriteRendererProperties}
   {rendererClass=&NiagaraSpriteRendererProperties, rendererIndex=1}
  ]
 }
 Emoji_AfterImage {
  bEnabled=false
  emitterUpdateScript [&EmitterState()]
  particleSpawnScript [&InitializeParticle(
    Color=&MultiplyLinearColor(
     Linear Color=&SelectLinearColorFromArray(Array Sampling Mode=::NewEnumerator0, Direct Array Index=?, Color Selection Array=$User.Color Array, Random Seed=0)
     Scale Factor=&MakeFloat4FromVectorAndFloat("Vector (RGB)"=&VectorFromFloat(Value=1), "Float (Alpha)"=1)
    )
    Lifetime=0.2
    Mass Mode=::NewEnumerator1
    Sprite Size Mode=::NewEnumerator3
    Uniform Sprite Size=100
    UsePositionOffset=false
    Write Material Random=true
   ) {module=InitializeParticle001}]
  particleUpdateScript [
   &ParticleState()
   &Drag(Drag=&RandomRangeFloat(Minimum=2, Maximum=4), Use Linear Drag=true, Use Rotational Drag=false)
   &AccelerationForce()
   &SolveForcesAndVelocity(Clamp Velocity=false, Limit Acceleration=false)
   &ScaleSpriteSize(Scale Factor=&Vector2DFromFloat(Value=&FloatFromCurve(
      FloatCurve=DataInterface() {propertyValues {Curve {keys [{tangentMode=RCTM_User, time=0, value=0.5}, {tangentMode=RCTM_User, time=1, value=1}]}}}
      Scale Curve=&FN_FrontEndFloatModifier(Gameplay=3, Front End=1)
     )))
   &ScaleSpriteSize(
    Scale Factor=&MakeVector2D(X=&Sine(Normalized Angle=$Emitter.Age, Period=0.7, Scale=1, Bias=0), Y=1)
    Scale Sprite Size Mode=::NewEnumerator2
   ) {enabled=false, module=ScaleSpriteSize001}
   &ScaleColor(
    Scale Alpha=&FloatFromCurve(FloatCurve=DataInterface() {propertyValues {Curve {keys [<at 0 1>,<at 1 0>]}}}, Scale Curve=0.2)
    Scale RGB=&VectorFromFloat(Value=&FN_FrontEndFloatModifier(Gameplay=1, Front End=0.25))
    ScaleA=true
    ScaleRGB=false
   )
  ]
  rendererClasses [&NiagaraSpriteRendererProperties]
  renderers [{rendererClass=&NiagaraSpriteRendererProperties}]
 }
 Emojis {
  bEnabled=false
  emitterUpdateScript [
   &EmitterState()
   &SpawnBurst_Instantaneous(Spawn Count=20, Use Spawn Probability=false)
  ]
  particleSpawnScript [
   &InitializeParticle(
    Color=<rgba 2 2 2 1>
    Lifetime=&Power(a=&RandomRangeFloat(Minimum=0.46, Maximum=1.5), b=2)
    Mass Mode=::NewEnumerator1
    Position Offset=&AddVectorToPosition(Position=$User.GlobalOffset, Vector=<0 0 2>)
    Position Offset Coordinate Space=::Local
    Sprite Rotation Angle=&RandomRangeFloat(Minimum=0, Maximum=360)
    Sprite Rotation Mode=::NewEnumerator0
    Sprite Size=&Vector2DFromFloat(Value=&RandomRangeFloat(Minimum=1, Maximum=5))
    Sprite Size Mode=::NewEnumerator4
    UsePositionOffset=true
    Write Material Random=true
   ) {module=InitializeParticle001}
   &SphereLocation(
    MaskSpawnGroup=false
    Override Seed=false
    Sphere Radius=&RandomRangeFloat(Minimum=4, Maximum=8)
    Surface Only=false
    UseNonUniformScale=false
    UseOffset=false
    UseSphereOrientation=false
   )
   &AddVelocityFromPoint(Module.Use Falloff Distance=false, Velocity Strength=&RandomRangeFloat(Minimum=50, Maximum=100))
   &AddVelocity(Velocity=&Multiply_VectorByFloat(Vector=$System.DirectionalVelocity, Float=0.4))
   &ApplyInitialForces()
   &SubUVAnimation(
    "Sprite Renderer "=DataInterface() {propertyValues {SpriteRenderer=&self:Emojis_5.NiagaraSpriteRendererProperties_1}}
    SubUV Animation Mode:ENiagara_SubUVLookupModeV2=::NewEnumerator2
    UseEndFrame=false
    UseStartFrame=false
   )
  ]
  particleUpdateScript [
   &ParticleState()
   &Drag(Drag=2, Use Linear Drag=true, Use Rotational Drag=false)
   &GravityForce(Gravity=&RandomRangeVector(Minimum=<0 0 -200>, Maximum=<1 1 -50>))
   &ScaleColor(Scale Alpha=&Lerp_Float(A=1, B=1, Alpha=$Particles.NormalizedAge), ScaleA=true, ScaleRGB=false)
   &ScaleSpriteSize(Scale Factor=&Vector2DFromFloat(Value=&FloatFromCurve(
      FloatCurve=DataInterface() {propertyValues {Curve {keys [<at 0.0018 0.1957>,<at 0.0527 0.819>,<at 0.1169 0.6555>,<at 0.1748 0.9015>,<at 0.2394 0.582>,<at 0.332 1.0023>,<at 0.4947 0.3707>,<at 0.6727 1>,<at 1 0>]}}}
      Scale Curve=4
     )))
   &ScaleSpriteSize(
    Scale Factor=&MakeVector2D(X=&Sine(Normalized Angle=$Emitter.Age, Period=0.7, Scale=1, Bias=0), Y=1)
    Scale Sprite Size Mode=::NewEnumerator2
   ) {module=ScaleSpriteSize001}
   &CurlNoiseForce(Noise Strength=200, Pan Noise=false)
   &InheritSourceMovement(
    Applied Position Falloff Distance=2000
    Applied Position Scale=$User.SourceMovement
    ApplyForce=false
    ApplyPosition=true
    ApplyPositionFalloff=false
   )
   &SolveForcesAndVelocity(Clamp Velocity=false, Limit Acceleration=false)
  ]
  rendererClasses [&NiagaraSpriteRendererProperties]
  renderers [{rendererClass=&NiagaraSpriteRendererProperties}]
 }
 EmojisLooping {
  bEnabled=false
  emitterUpdateScript [
   &EmitterState(
    Delay First Loop Only=true
    Life Cycle Mode=::NewEnumerator1
    Loop Behavior=::NewEnumerator2
    Loop Count=43
    Loop Delay=2.2
    Loop Duration=0.2
    UseLoopDelay=true
   )
   &SpawnBurst_Instantaneous(
    Spawn Count=&RandomRangeInt(Minimum=1, Maximum=1, Evaluation Type=::NewEnumerator1)
    Spawn Time=0.15
    Use Spawn Probability=false
   )
  ]
  particleSpawnScript [
   &FN_SetSpawnGroup(Number of Spawn Groups=6)
   &InitializeParticle(
    Color=<rgba 2 2 2 1>
    Lifetime=&RandomRangeFloat(Minimum=1, Maximum=1.6)
    Mass Mode=::NewEnumerator1
    Position Offset=<0 0 40>
    Position Offset Coordinate Space=::Local
    Sprite Rotation Angle=&RandomRangeFloat(Minimum=0, Maximum=360)
    Sprite Rotation Mode=::NewEnumerator0
    Sprite Size=&Vector2DFromFloat(Value=&RandomRangeFloat(Minimum=1, Maximum=5))
    Sprite Size Mode=::NewEnumerator4
    UsePositionOffset=true
    Write Material Random=true
   ) {module=InitializeParticle001}
   &ShapeLocation(MaskSpawnGroup=false, Override Seed=false, Sphere Radius=40, UseOwnerScale=true)
   &AddVelocity(Module.Use Falloff Distance=false, Override Seed=false, Velocity Mode=::NewEnumerator1, Velocity Speed=400) {enabled=false}
   /SetVariables(Particles.InitialSize:Vector2f=&Vector2DFromFloat(Value=&RandomRangeFloat(Minimum=40, Maximum=80))) {isSetParameters=true, module=SetVariables_#2}
   /SetVariables(Particles.SpriteRotation2:NiagaraFloat=&RandomRangeFloat(Minimum=0, Maximum=360)) {isSetParameters=true, module=SetVariables_#3}
   &CameraOffset(Camera Offset Amount=&RandomRangeFloat(Minimum=20, Maximum=40))
   &SubUVAnimation(
    Output Alternate Renderer Binding=true
    "Sprite Renderer "=DataInterface() {propertyValues {SpriteRenderer=&self:EmojisLooping_5.NiagaraSpriteRendererProperties_2}}
    SubUV Animation Mode:ENiagara_SubUVLookupModeV2=::NewEnumerator2
    UseEndFrame=false
    UseStartFrame=false
   )
  ]
  particleUpdateScript [
   &ParticleState()
   &SpriteRotationRate(Rotation Rate=100)
   &Drag(Drag=2, Use Linear Drag=true, Use Rotational Drag=false)
   &GravityForce(Gravity=&RandomRangeVector(Minimum=<0 0 -200>, Maximum=<1 1 -50>))
   &ScaleColor(Scale Alpha=&Lerp_Float(A=1, B=1, Alpha=$Particles.NormalizedAge), ScaleA=true, ScaleRGB=false)
   &ScaleSpriteSize(Scale Factor=&Vector2DFromFloat(Value=&FloatFromCurve(
      FloatCurve=DataInterface() {propertyValues {Curve {keys [
          <at 0.0018 0.1957>
          <at 0.0527 0.819>
          <at 0.1169 0.6555>
          <at 0.1748 0.9015>
          <at 0.2394 0.582>
          <at 0.332 1.0023>
          <at 0.4947 0.3707>
          {interpMode=RCIM_Cubic, tangentMode=RCTM_User, time=0.6727, value=1}
          {arriveTangent=-2.0158, interpMode=RCIM_Cubic, leaveTangent=-2.0158, tangentMode=RCTM_User, time=1, value=0.5}
         ]}}}
      Scale Curve=4
     )))
   &ScaleSpriteSize(
    Scale Factor=&MakeVector2D(X=&Sine(Normalized Angle=$Emitter.Age, Period=1, Scale=1, Bias=0), Y=1)
    Scale Sprite Size Mode=::NewEnumerator2
   ) {module=ScaleSpriteSize001}
   &CurlNoiseForce(Noise Quality / Cost=::NewEnumerator0, Noise Strength=100, Pan Noise=false)
   &InheritSourceMovement(Applied Position Falloff Distance=2000, ApplyForce=false, ApplyPosition=true, ApplyPositionFalloff=false) {enabled=false}
   &SolveForcesAndVelocity(Clamp Velocity=false, Limit Acceleration=false)
   &KillParticles(Kill Particles=&SetBoolByFloatComparison(
     A=&FloatFromCurve(
      FloatCurve=DataInterface() {propertyValues {Curve {keys [<at 0 0 RCIM_Constant>,<at 0.99 1.5 RCIM_Constant>]}}}
      CurveIndex=&Divide_Float(A=$System.Age, B=$User.Duration)
     )
     B=1
    ))
   &GenerateDeathEvent(
    "Float to Send as Float 1 (NormalizedAge)":NiagaraFloat=$Particles.SubUVAnimation.SubImageIndex
    "Float to Send as Float 2 (RandomNormalizedFloat)":NiagaraFloat=$Particles.SpriteRotation
    "Float to Send as Float 3 (DistanceTraveled)":NiagaraFloat=&MakeFloatFromVector2d(Vector 2D=$Particles.SpriteSize, Channel=::NewEnumerator1)
    Use Event Probability=false
   )
  ]
  rendererClasses [&NiagaraSpriteRendererProperties]
  renderers [{rendererClass=&NiagaraSpriteRendererProperties}]
 }
 GlowInsideMesh {
  emitterUpdateScript [
   &EmitterState()
   &SpawnBurst_Instantaneous(Spawn Count=1, Use Spawn Probability=false)
  ]
  particleSpawnScript [
   &InitializeParticle(
    Color=<rgba 200 104 40 0.1>
    Lifetime=$User.Duration
    Mass Mode=::NewEnumerator1
    Mesh Scale=&Multiply_Vector(A=<0.85 1.2 1.1>, B=&VectorFromFloat(Value=0.2))
    Mesh Scale Mode=::NewEnumerator4
    Position Offset=&AddVectorToPosition(Position=$User.GlobalOffset, Vector=<0 0 0>)
    Position Offset Coordinate Space=::Local
    UsePositionOffset=true
    Write Material Random=true
   ) {module=InitializeParticle001}
   &InitialMeshOrientation(Mesh Orientation Mode=::NewEnumerator7, Rotation=<0 0 0.25>, Rotation Coordinate Space=::NewEnumerator2, Use Rotation Vector=true) {module=InitialMeshOrientation001}
  ]
  particleUpdateScript [&ParticleState()]
  rendererClasses [&NiagaraMeshRendererProperties]
  renderers [{rendererClass=&NiagaraMeshRendererProperties, rendererIndex=0}]
 }
 GroundLightMesh {
  emitterUpdateScript [
   &EmitterState(
    Inactive Response=::NewEnumerator1
    Life Cycle Mode=::NewEnumerator1
    Loop Behavior=::NewEnumerator1
    Loop Duration Mode=::NewEnumerator1
    UseLoopDelay=false
   )
   &SpawnBurst_Instantaneous(Spawn Count=2, Use Spawn Probability=false, UseLoopCountLimit=false)
  ]
  particleSpawnScript [
   &FN_SetSpawnGroup(Number of Spawn Groups=6)
   &InitializeParticle(
    Color=<rgba 1 0.797 0.3287 0.5>
    Lifetime Max=3
    Lifetime Min=1.5
    Lifetime Mode=::NewEnumerator1
    Mesh Scale=<0.1 0.1 0.3>
    Mesh Scale Mode=::NewEnumerator4
    Position Offset=&Lerp_Vector(A=<0 15 0>, B=<0 -15 0>, Alpha=&ReturnNormalizedExecIndex())
    Position Offset Coordinate Space=::Local
    UsePositionOffset=true
    Write Material Random=true
   )
   &InitialMeshOrientation(Mesh Orientation Mode=::NewEnumerator7, Rotation=<0.25 0.5 0>, Use Rotation Vector=true)
  ]
  particleUpdateScript [
   &ParticleState(
    Kill Particles When Lifetime Has Elapsed=false
    Let Infinitely Lived Particles Die When Emitter Deactivates=true
    Loop Particles Lifetime=true
   )
   &UpdateMeshOrientation(
    Rotation Coordinate Space=::NewEnumerator0
    Rotation Rate=&SelectFloatFromArray(Direct Array Index=$Particles.SpawnGroup, Float Selection Array=DataInterface() {propertyValues {FloatData [0.3,-0.4,0.35,-0.38,0.4,-0.35]}})
    Rotation Vector=<0 0 1>
   )
   &Color(Color=&ColorFromCurve(
     ColorCurve=DataInterface() {propertyValues {
       AlphaCurve {keys [<at 0 0.75 RCIM_Cubic>]}
       BlueCurve {keys [<at 0.1488 1.1081 RCIM_Cubic>]}
       GreenCurve {keys [<at 0.1488 2.8534 RCIM_Cubic>]}
       RedCurve {keys [<at 0.1488 4 RCIM_Cubic>]}
      }}
     CurveIndex=$System.NormalizedLoopAge
    )) {module=Color001}
   &ScaleColor(
    Color Value To Scale=$Particles.Color
    Scale Alpha=&Multiply_Float(
     A=&FloatFromCurve(
      FloatCurve=DataInterface() {propertyValues {Curve {keys [<at 0 1>,<at 0.2 0.5>,<at 0.4 0.89>,<at 0.5 0.45>,<at 0.7 0.8>,<at 0.9 0.7>,<at 1 1>]}}}
      CurveIndex=&Frac_Float(Float=&Add_Float(
        A=$Emitter.Age
        B=&SelectFloatFromArray(Direct Array Index=$Particles.SpawnGroup, Float Selection Array=DataInterface() {propertyValues {FloatData [0,0.3,0.5,0.4,0.77,0.18]}})
       ))
     )
     B=&FN_FrontEndFloatModifier(Gameplay=2, Front End=2)
    )
    ScaleA=true
    ScaleRGB=true
   )
   &ScaleColor(
    Scale Alpha=&FloatFromCurve(
     FloatCurve=DataInterface() {propertyValues {Curve {keys [<at 0 1>,<at 1 0>]}}}
     CurveIndex=&Subtract_Float(A=$Emitter.Age, B=11)
    )
    ScaleA=true
    ScaleRGB=false
   ) {module=ScaleColor001}
   &ScaleColor(
    Scale Alpha=&FloatFromCurve(FloatCurve=DataInterface() {propertyValues {Curve {keys [<at 2.25 0>,<at 2.7 1>]}}}, CurveIndex=$Engine.System.Age)
    ScaleA=true
    ScaleRGB=false
   ) {module=ScaleColor002}
  ]
  rendererClasses [&NiagaraMeshRendererProperties, &NiagaraSpriteRendererProperties]
  renderers [
   {rendererClass=&NiagaraMeshRendererProperties, rendererIndex=0}
   {rendererClass=&NiagaraSpriteRendererProperties, rendererIndex=1}
  ]
 }
 LightRays {
  emitterSpawnScript [/SetVariables(Emitter.RayDuration=0.4) {isSetParameters=true}]
  emitterUpdateScript [
   &EmitterState(
    Delay First Loop Only=true
    Life Cycle Mode=::NewEnumerator1
    Loop Behavior=::NewEnumerator2
    Loop Count=22
    Loop Delay=2.25
    Loop Duration=$Emitter.RayDuration
    UseLoopDelay=true
   )
   &SpawnBurst_Instantaneous(Spawn Group=1, Use Spawn Probability=false, UseLoopCountLimit=false) {module=SpawnBurst_Instantaneous001}
   &SpawnBurst_Instantaneous(Spawn Count=1, Use Spawn Probability=false)
  ]
  particleSpawnScript [
   &InitializeParticle(
    Color=&MaskLinearColorBySpawnGroup(Masked Linear Color=<rgba 2 1.3469 0.4 1>, Passthrough Linear Color=<rgba 2 1.7143 1.3 1>)
    Lifetime=$Emitter.RayDuration
    Position Offset=&AddVectorToPosition(Position=$User.GlobalOffset, Vector=<0 0 0>)
    Position Offset Coordinate Space=::Local
    Sprite Rotation Angle=&UniformRangedFloat(Minimum=0, Maximum=0)
    Sprite Rotation Mode=::NewEnumerator0
    Sprite Size=&MaskVector2dBySpawnGroup(
     Masked Vector2D=&UniformRangedVector2D(Minimum=<80 50>, Maximum=<120 60>)
     Passthrough Vector2D=&UniformRangedVector2D(Minimum=<55 30>, Maximum=<70 30>)
     Spawn Group Mask=0
    )
    Sprite Size Mode=::NewEnumerator4
    UsePositionOffset=true
    Write Material Random=true
   ) {module=InitializeParticle001}
   /Niagara/Modules/Update/SubUV/SubUVAnimation(SubUV Animation Mode=::NewEnumerator1)
   &AddVelocity(Velocity=&RandomRangeVector(Minimum=<0 0 1>, Maximum=<0 0 1>))
   &SystemLocation(MaskSpawnGroup=false, Offset=<0 0 -5>)
  ]
  particleUpdateScript [
   &ParticleState()
   &ScaleSpriteSize(Scale Factor=&Vector2DFromFloat(Value=&FloatFromCurve(FloatCurve=DataInterface() {propertyValues {Curve {keys [
          <at -0.0017 0.0417>
          {arriveTangent=1.1486, leaveTangent=1.1486, time=0.03, value=1}
          {arriveTangent=-0.1439, interpMode=RCIM_Cubic, leaveTangent=-0.1439, tangentMode=RCTM_User, time=1, value=1.0006}
         ]}}})))
   &AccelerationForce(Acceleration=&VectorFromCurve(VectorCurve=DataInterface() {propertyValues {XCurve {keys [<at 0 0>,<at 1 0>]}, YCurve {keys [<at 0 0>,<at 1 0>]}, ZCurve {keys [<at 0 0>,<at 1 1>]}}}))
   &SolveForcesAndVelocity(Clamp Velocity=false, Limit Acceleration=false)
   &ScaleColor(
    Scale Alpha=&FloatFromCurve(FloatCurve=DataInterface() {propertyValues {Curve {keys [<at 0 0>,<at 0.01 1>,<at 0.9 1>,<at 1 0>]}}})
    ScaleA=true
    ScaleRGB=true
   )
   &ScaleColor(
    Scale Alpha=&FloatFromCurve(
     FloatCurve=DataInterface() {propertyValues {Curve {keys [
         {interpMode=RCIM_Cubic, tangentMode=RCTM_User, time=0, value=0}
         {interpMode=RCIM_Cubic, tangentMode=RCTM_User, time=0.5, value=1}
         {tangentMode=RCTM_User, time=1, value=0}
        ]}}}
     Scale Curve=0.5
    )
    ScaleA=true
    ScaleRGB=true
   ) {module=ScaleColor001}
  ]
  rendererClasses [&NiagaraSpriteRendererProperties]
  renderers [{rendererClass=&NiagaraSpriteRendererProperties}]
 }
 Lightray_Ending {
  emitterUpdateScript [
   &EmitterState(Life Cycle Mode=::NewEnumerator1, Loop Behavior=::NewEnumerator1, Loop Delay=10.9, Loop Duration=16, UseLoopDelay=true)
   &SpawnBurst_Instantaneous(Spawn Group=1, Use Spawn Probability=false, UseLoopCountLimit=false) {module=SpawnBurst_Instantaneous001}
   &SpawnBurst_Instantaneous(Spawn Count=1, Use Spawn Probability=false)
  ]
  particleSpawnScript [
   &InitializeParticle(
    Color=&MaskLinearColorBySpawnGroup(Masked Linear Color=<rgba 2 1.3469 0.4 1>, Passthrough Linear Color=<rgba 2 1.7143 1.3 1>)
    Position Offset=&AddVectorToPosition(Position=$User.GlobalOffset, Vector=<0 0 0>)
    Position Offset Coordinate Space=::Local
    Sprite Rotation Angle=&UniformRangedFloat(Minimum=0, Maximum=0)
    Sprite Rotation Mode=::NewEnumerator0
    Sprite Size=&MaskVector2dBySpawnGroup(
     Masked Vector2D=&UniformRangedVector2D(Minimum=<80 60>, Maximum=<120 60>)
     Passthrough Vector2D=&UniformRangedVector2D(Minimum=<55 30>, Maximum=<70 30>)
     Spawn Group Mask=0
    )
    Sprite Size Mode=::NewEnumerator4
    UsePositionOffset=true
    Write Material Random=true
   ) {module=InitializeParticle001}
   &AddVelocity()
   &SubUVAnimation(
    Random SubUV Lookup Index:NiagaraFloat=&FloatFromCurve(
     FloatCurve=DataInterface() {propertyValues {Curve {keys [<at 0 0>,<at 1 1>]}}}
     CurveIndex=&RandomRangeFloat(Minimum=0, Maximum=1)
    )
    SubUV Animation Mode=::NewEnumerator1
   )
   &SystemLocation(MaskSpawnGroup=false, Offset=<0 0 -5>)
  ]
  particleUpdateScript [
   &ParticleState()
   &SolveForcesAndVelocity(Clamp Velocity=false, Limit Acceleration=false)
   &ScaleSpriteSize(Scale Factor=&Vector2DFromFloat(Value=&FloatFromCurve(FloatCurve=DataInterface() {propertyValues {Curve {keys [
          <at -0.0017 0.0417>
          {arriveTangent=1.1486, leaveTangent=1.1486, time=0.03, value=1}
          {arriveTangent=-0.1439, interpMode=RCIM_Cubic, leaveTangent=-0.1439, tangentMode=RCTM_User, time=1, value=1.0006}
         ]}}})))
   &ScaleColor(
    Scale Alpha=&FloatFromCurve(FloatCurve=DataInterface() {propertyValues {Curve {keys [<at 0 0>,<at 0.01 1>,<at 0.9 1>,<at 1 0>]}}})
    ScaleA=true
    ScaleRGB=true
   ) {enabled=false}
   &ScaleColor(
    Scale Alpha=&FloatFromCurve(FloatCurve=DataInterface() {propertyValues {Curve {keys [
         {interpMode=RCIM_Cubic, tangentMode=RCTM_User, time=0, value=0}
         {interpMode=RCIM_Cubic, tangentMode=RCTM_User, time=0.5, value=1}
         {tangentMode=RCTM_User, time=1, value=0}
        ]}}})
    ScaleA=true
    ScaleRGB=true
   ) {module=ScaleColor001}
  ]
  rendererClasses [&NiagaraSpriteRendererProperties]
  renderers [{rendererClass=&NiagaraSpriteRendererProperties}]
 }
 Sparkles {
  emitterUpdateScript [
   &EmitterState()
   &SpawnBurst_Instantaneous(Spawn Count=22, Use Spawn Probability=false)
  ]
  particleSpawnScript [
   &InitializeParticle(
    Color=<rgba 5 3.7043 1.51 1>
    Lifetime=&UniformRangedFloat(Minimum=0.3, Maximum=1)
    Position Offset=&RandomRangeVector(Minimum=<-5 -10 -5>, Maximum=<5 10 0>)
    Position Offset Coordinate Space=::Local
    Sprite Size Mode=::NewEnumerator3
    Uniform Sprite Size=&UniformRangedFloat(Minimum=5, Maximum=8)
    UsePositionOffset=true
    Write Material Random=true
   ) {module=InitializeParticle001}
   &AddVelocity(Velocity=&UniformRangedVector(Minimum=<-100 -100 -10>, Maximum=<100 100 225>))
   &CameraOffset(Camera Offset Amount=-25)
   &ApplyInitialForces()
  ]
  particleUpdateScript [
   &ParticleState()
   &GravityForce(Gravity=<0 0 50>)
   /Niagara/Modules/Update/Forces/CurlNoiseForce(Noise Frequency=0.5, Noise Quality / Cost=::NewEnumerator0, Noise Strength=150, Pan Noise=true, Pan Noise Field=<0 0 0.25>)
   &Drag(Drag=5, Use Linear Drag=true, Use Rotational Drag=false)
   &ScaleColor(
    Scale Alpha=&FloatFromCurve(FloatCurve=DataInterface() {propertyValues {Curve {keys [<at -0.0092 0.0144>,<at 0.0967 0.9767>,<at 1.0011 0.002>]}}})
    ScaleA=true
    ScaleRGB=true
   )
   &ScaleSpriteSize(Scale Factor=&Vector2DFromFloat(Value=&FloatFromCurve(FloatCurve=DataInterface() {propertyValues {Curve {keys [<at -0.0046 0>,<at 0.1746 0.9952>,<at 1.0011 0.002>]}}})))
   &InheritSourceMovement(
    Applied Position Falloff Distance=2000
    Applied Position Scale=$User.SourceMovement
    ApplyForce=false
    ApplyPosition=true
    ApplyPositionFalloff=false
   )
   &SolveForcesAndVelocity(Clamp Velocity=false, Limit Acceleration=false)
  ]
  rendererClasses [&NiagaraSpriteRendererProperties]
  renderers [{rendererClass=&NiagaraSpriteRendererProperties}]
 }
 Sparkles001 {
  emitterUpdateScript [
   &EmitterState()
   &SpawnBurst_Instantaneous(Spawn Count=10, Use Spawn Probability=false)
  ]
  particleSpawnScript [
   /Niagara/Modules/Spawn/Initialization/InitializeParticle(
    Color=&HueShiftLinearColor(
     Hue=0
     Saturation=&RandomRangeFloat(Minimum=5, Maximum=10)
     Value=&RandomRangeFloat(Minimum=1, Maximum=10)
     Linear Color=&RandomRangeLinearColor(Minimum=<rgba 0 0 0 1>, Maximum=<rgba 1 1 1 1>, Color Channel Mode=::NewEnumerator2)
    )
    Lifetime=&Power(a=&RandomRangeFloat(Minimum=0.46, Maximum=1.5), b=2)
    Module.Write Color=true
    Module.Write Lifetime=true
    Module.Write Mass=true
    Module.Write Material Random=true
    Module.Write Position=false
    Module.Write Scale=false
    Module.Write SpriteRotation=true
    Module.Write SpriteSize=true
    Position=&SimulationPosition()
    Sprite Rotation=&RandomRangeFloat(Minimum=0, Maximum=360)
    Sprite Size=&Vector2DFromFloat(Value=&RandomRangeFloat(Minimum=1, Maximum=5))
   ) {enabled=false}
   /Niagara/Modules/Spawn/Initialization/V2/InitializeParticle(
    Color=&ColorFromCurve(
     ColorCurve=DataInterface() {propertyValues {
       AlphaCurve {keys [<at 0 1 RCIM_Cubic>,<at 1 1 RCIM_Cubic>]}
       BlueCurve {keys [<at 0 0 RCIM_Cubic>,<at 0.479 0.436>,<at 1 0 RCIM_Cubic>]}
       GreenCurve {keys [<at 0 0.7843 RCIM_Cubic>,<at 0.479 1>,<at 1 0.1355 RCIM_Cubic>]}
       RedCurve {keys [<at 0 1 RCIM_Cubic>,<at 0.479 0.1729>,<at 1 1 RCIM_Cubic>]}
      }}
     CurveIndex=&RandomRangeFloat(Minimum=0, Maximum=1)
    )
    Lifetime=&Power(a=&RandomRangeFloat(Minimum=0.46, Maximum=1.5), b=2)
    Mass Mode=::NewEnumerator1
    Position Offset=&AddVectorToPosition(Position=$User.GlobalOffset, Vector=<0 0 0>)
    Position Offset Coordinate Space=::Local
    Sprite Rotation Angle=&RandomRangeFloat(Minimum=0, Maximum=360)
    Sprite Rotation Mode=::NewEnumerator0
    Sprite Size=&Vector2DFromFloat(Value=&RandomRangeFloat(Minimum=1, Maximum=5))
    Sprite Size Mode=::NewEnumerator4
    UsePositionOffset=true
    Write Material Random=true
   ) {module=InitializeParticle001}
   &SphereLocation(
    MaskSpawnGroup=false
    Override Seed=false
    Sphere Radius=&RandomRangeFloat(Minimum=1, Maximum=4)
    Surface Only=false
    UseNonUniformScale=false
    UseOffset=false
    UseSphereOrientation=false
   )
   &AddVelocityFromPoint(Module.Use Falloff Distance=false, Velocity Strength=&RandomRangeFloat(Minimum=50, Maximum=150))
   &AddVelocity(Velocity=&Multiply_VectorByFloat(Vector=$System.DirectionalVelocity, Float=1))
   &InheritVelocity(
    Inherited Velocity Amount Scale=<10 10 10>
    Inherited Velocity Speed Limit:NiagaraFloat=$User.InheritVelocity
    Speed Limit=true
    UseSpeedThreshold=false
   ) {enabled=false}
   &ApplyInitialForces()
  ]
  particleUpdateScript [
   &ParticleState()
   &Drag(Drag=3, Use Linear Drag=true, Use Rotational Drag=false)
   &GravityForce(Gravity=&RandomRangeVector(Minimum=<0 0 -200>, Maximum=<1 1 -50>))
   &ScaleColor(
    Scale Alpha=&Lerp_Float(A=1, B=0, Alpha=$Particles.NormalizedAge)
    Scale RGB=&VectorFromFloat(Value=5)
    ScaleA=true
    ScaleRGB=true
   )
   &ScaleSpriteSize(Scale Factor=&Vector2DFromFloat(Value=&FloatFromCurve(
      FloatCurve=DataInterface() {propertyValues {Curve {keys [<at 0.0018 0.1957>,<at 0.0527 0.819>,<at 0.1169 0.6555>,<at 0.1748 0.9015>,<at 0.2394 0.582>,<at 0.332 1.0023>,<at 0.4947 0.3707>,<at 0.6727 1>,<at 1 0>]}}}
      Scale Curve=2
     )))
   &InheritSourceMovement(
    Applied Position Falloff Distance=2000
    Applied Position Scale=$User.SourceMovement
    ApplyForce=false
    ApplyPosition=true
    ApplyPositionFalloff=false
   )
   &SolveForcesAndVelocity(Clamp Velocity=false, Limit Acceleration=false)
  ]
  rendererClasses [&NiagaraSpriteRendererProperties]
  renderers [{rendererClass=&NiagaraSpriteRendererProperties}]
 }
}
summary {
 systemName=NS_SilverbellEmote_FX
 emitters [
  {emitterName=Sparkles, rendererClasses [&NiagaraSpriteRendererProperties]}
  {emitterName=Sparkles001, rendererClasses [&NiagaraSpriteRendererProperties]}
  {emitterName=LightRays, rendererClasses [&NiagaraSpriteRendererProperties]}
  {emitterName=GlowInsideMesh, rendererClasses [&NiagaraMeshRendererProperties]}
  {emitterName=Airglow, rendererClasses [&NiagaraSpriteRendererProperties, &NiagaraLightRendererProperties]}
  {emitterName=Lightray_Ending, rendererClasses [&NiagaraSpriteRendererProperties]}
  {bEnabled=false, emitterName=Emojis, rendererClasses [&NiagaraSpriteRendererProperties]}
  {emitterName=Bling_Big, rendererClasses [&NiagaraSpriteRendererProperties]}
  {bEnabled=false, emitterName=EmojisLooping, rendererClasses [&NiagaraSpriteRendererProperties]}
  {bEnabled=false, emitterName=Emoji_AfterImage, rendererClasses [&NiagaraSpriteRendererProperties]}
  {emitterName=Bling_Big001, rendererClasses [&NiagaraSpriteRendererProperties]}
  {
   emitterName=GroundLightMesh
   rendererClasses [&NiagaraMeshRendererProperties, &NiagaraSpriteRendererProperties]
  }
 ]
 userVariables [
  User.bIsFrontEnd {type=NiagaraBool}
  User.bIsFrontEndPreview {type=NiagaraBool}
  User.Color Array {
   type=NiagaraDataInterfaceArrayColor
   description=""
   defaultValue {
    struct=NiagaraExt_VariableValue_DataInterface
    value {
     dataInterface=&self:NiagaraDataInterfaceArrayColor_0
     dataInterfaceClass=&NiagaraDataInterfaceArrayColor
    }
   }
  }
  User.Duration {type=NiagaraFloat, defaultValue=12, description=""}
  User.GlobalOffset {type=NiagaraPosition, description="", defaultValue {}}
  User.SourceMovement {type=Vector3f, description="", defaultValue {struct=Vector3f, value=<0.6 0.6 0.25>}}
 ]
}
